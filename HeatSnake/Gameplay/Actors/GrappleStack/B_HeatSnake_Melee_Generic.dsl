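; Generic Fortnite melee weapon Blueprint that handles cosmetic effects, impact audio and force feedback.
; Construction assigns idle, swing and trail Niagara templates, and equipping binds swing and trail callbacks through the player's interface.
; Impact events reposition spawned FX from pawn and weapon sockets and play pickaxe audio for hostile wielders.
; It also attaches elemental alteration particles and shows durability-break effects with the item's icon on a dynamic material.
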(event EventBeginPlay
  ;[1 1 1] Event Begin Play
  (Effects|Components|ParticleSystem|SetAutoAttachmentParameters (Variables|Default|GetIdleFX%20Niagara) (Variables|Visual|GetWeaponMesh) (Variables|Effects|GetIdleFXAttachSocket) (Variables|Effects|GetIdleFX%20Location%20Rule) (Variables|Effects|GetIdle%20FX%20Rotation%20Rule) (Variables|Effects|GetIdle%20FX%20Scale%20Rule))
  (Niagara|ResetSystem (Variables|Default|GetIdleFX%20Niagara))
  (Effects|Components|ParticleSystem|SetAutoAttachmentParameters (Variables|Default|GetSwingFX%20Niagara) (Variables|Visual|GetWeaponMesh) (Variables|Effects|GetSwingFXSocket) (Variables|Effects|GetSwingFX%20Location%20Rule) (Variables|Effects|GetSwing%20FX%20Rotation%20Rule) (Variables|Effects|GetSwing%20FX%20Scale%20Rule))
  (Niagara|ResetSystem (Variables|Default|GetSwingFX%20Niagara))
  (|CallParentFunction))

(event Custom|On%20Anim%20Trails%20Notify
  (params (bActive :type Boolean))
  ;[1 1 1] NOTE: This currently ONLY supports Niagara, since we shouldn't be creating Cascade Anim Trails FX any more.
  (if bActive
    (Networking|CanExecuteCosmeticEvents
      (:True
        (Components|Activation|Activate (Variables|Default|GetAnimTrail%20Niagara) true)))
    (else
      (Networking|CanExecuteCosmeticEvents
        (:True
          (Components|Activation|Deactivate (Variables|Default|GetAnimTrail%20Niagara)))))))

(event FortWeapon|EventOnEquippedAndAttached
  ;[1 1 1] New Anim Trails
  (CallFunction|BindSwingAndAnimTrailEvents)
  (CallFunction|SetActiveAlterationIdleParticles true))

(event Custom|On%20Anim%20Trails%20Notify%20%28Deactivate%29
  ;[1 1 1] NOTE: This currently ONLY supports Niagara, since we shouldn't be creating Cascade Anim Trails FX any more.
  (Networking|CanExecuteCosmeticEvents
    (:True
      (Components|Activation|Deactivate (Variables|Default|GetAnimTrail%20Niagara)))))

(event FortWeapon|EventOnUnEquippedAndDetached
  ;[1 1 1] New Anim Trails
  (CallFunction|UnbindSwingandAnimTrailEvents)
  (CallFunction|SetActiveAlterationIdleParticles false)
  (bind _asfort_player_pawn (Utilities|Casting|CastToFortPlayerPawn (Game|GetInstigator)))
  (Animation|AnimTrailsNotify _asfort_player_pawn false)
  (Animation|AnimTrailsDisable _asfort_player_pawn))

(event Custom|OnSwingRight
  ;[1 1 1] NOTE: This currently only support Niagara, since we shouldn't be creating Cascade Swing FX any more.
  (CallFunction|ActivateSwingFX))

(event Custom|OnSwingRightEnded
  ;[1 1 1] NOTE: This currently only support Niagara, since we shouldn't be creating Cascade Swing FX any more.
  (CallFunction|DeactivateSwingFX))

(event FortWeapon|EventOnLocalUnEquip
  ;[1 1 1] New Anim Trails
  (CallFunction|UnbindSwingandAnimTrailEvents))

(event Custom|OnSwingLeft
  ;[1 1 1] NOTE: This currently only support Niagara, since we shouldn't be creating Cascade Swing FX any more.
  (CallFunction|ActivateSwingFX))

(event Custom|OnSwingLeftEnded
  ;[1 1 1] NOTE: This currently only support Niagara, since we shouldn't be creating Cascade Swing FX any more.
  (CallFunction|DeactivateSwingFX))

(event EventOnPlayWeaponFireFX
  (params (bPersistentFire :type Boolean) (bSecondaryFire :type Boolean))
  ;[1 1 1] Event on Play Weapon Fire FX - called from Combat Melee Montage AnimNotify PlayFireFX
  (FortWeapon|PlayForceFeedback false))

(event Custom|PlayRClickImpacts
  ;[1 1 1] Spawn RClick Impacts from AnimNotify
  (Components|Activation|Activate (Variables|Default|GetMeleeHeavy_PSC) false))

(event EventOnInitCosmeticAlterations
  (params (CosmeticMod :type Fort%20Cosmetic%20Modification%20Structure))
  (bind (_cosmeticmaterial _ambientparticlesystem _ambientniagarasystem) (Utilities|Struct|BreakFortCosmeticModification CosmeticMod))
  ;[1 1 1] Cosmetic Alterations
  ;; Elemental Alteration
  (Networking|CanExecuteCosmeticEvents
    (:True
      (Utilities|FlowControl|Sequence
        (:then_0
          (if (not (Asset|IsAssetNull _ambientniagarasystem))
            (bind _object (Utilities|AsyncLoadAsset _ambientniagarasystem)
              (:Completed
                (bind _asniagara_system (Utilities|Casting|CastToNiagaraSystem _object))
                (bind _niagara_component_system_reference (Utilities|ElementalWeaponSetupFX (Variables|Visual|GetWeaponMesh) _asniagara_system "FX_Idle_Start" "FX_Idle_End" false)
                  (:Out
                    (bind _output_get (Variables|Default|SetAlteration%20Ambient%20PS _niagara_component_system_reference))
                    (LOD|SetMaxDrawDistance _output_get 5e+03)))))))
        (:then_1
          (bind _asplayer_pawn_generic_parent (Utilities|Casting|CastToPlayerPawn_Generic_Parent (Game|GetInstigator self)))
          (Class|PlayerPawnGenericParent|ReinitializeWeaponMaterials _asplayer_pawn_generic_parent))))))

(event FortWeapon|EventOnPlayImpactFX
  (params (HitResult :type Hit%20Result%20Structure%20%28by%20ref%29) (ImpactPhysicalSurface :type EPhysicalSurface%20Enum :default "SurfaceType_Default") (SpawnedPSC :type FXSystem%20Component%20Object%20Reference))
  (bind (_bblockinghit _binitialoverlap _time _distance _location _impactpoint _normal _impactnormal _physmat _hitactor) (Collision|BreakHitResult HitResult))
  (bind _returnvalue (Transformation|GetActorLocation _hitactor))
  (bind _returnvalue_2 (- _returnvalue "0, 0, 0"))
  (bind _x (.x _returnvalue_2))
  (bind _y (.y _returnvalue_2))
  (bind _z (.z _returnvalue_2))
  (bind _returnvalue_3 (Math|Vector|MakeVector _x _y _z))
  (bind _returnvalue_4 (Math|Vector|Normalize _returnvalue_3))
  ;[1 1 1] Impact FX
  (|CallParentFunction HitResult ImpactPhysicalSurface SpawnedPSC)
  (FortWeapon|PlayForceFeedback (Math|Boolean|ORBoolean (Utilities|Enum|Equal(Enum) ImpactPhysicalSurface "SurfaceType14") (Utilities|Enum|Equal(Enum) ImpactPhysicalSurface "SurfaceType15") (Utilities|Enum|Equal(Enum) ImpactPhysicalSurface "SurfaceType16")) false true)
  (CallFunction|PlayCQCPickaxeEnemyAudio HitResult)
  (if (Math|Boolean|ORBoolean (Utilities|Enum|Equal(Enum) ImpactPhysicalSurface "SurfaceType6") (Utilities|Enum|Equal(Enum) ImpactPhysicalSurface "SurfaceType7") (Utilities|Enum|Equal(Enum) ImpactPhysicalSurface "SurfaceType4"))
    (bind _asfort_pawn (Utilities|Casting|CastToFortPawn (Game|GetInstigator _hitactor)))
    (bind _outsockettransform (CallFunction|GetWeaponMeshSocketTransform "SwingFX"))
    (bind (_outsockettransform_location _outsockettransform_rotation) (Math|Transform|BreakTransform _outsockettransform))
    (Utilities|IsValid (Components|SkeletalMesh|GetAnimInstance (Variables|/Script/Engine.Character|GetMesh _asfort_pawn))
      (:Is%20Valid
        (Utilities|IsValid SpawnedPSC
          (:Is%20Valid
            (if (Utilities|Array|ContainsItem (Variables|Effects|GetPokeAnimations) (Animation|Montage|GetCurrentActiveMontage (Components|SkeletalMesh|GetAnimInstance (Variables|/Script/Engine.Character|GetMesh _asfort_pawn))))
              (Effects|Components|ParticleSystem|SetEmitterEnable SpawnedPSC "Sparks" false)
              (Effects|Components|ParticleSystem|SetEmitterEnable SpawnedPSC "Swoosh" false)
              (else
                (Utilities|IsValid SpawnedPSC
                  (:Is%20Valid
                    _))))
            (Transformation|SetWorldLocationAndRotation SpawnedPSC (select (Utilities|Enum|Equal(Enum) ImpactPhysicalSurface "SurfaceType4") (Math|Vector|Lerp(Vector) (Transformation|GetSocketLocation (Variables|/Script/Engine.Character|GetMesh _asfort_pawn) "FX_Chest") _outsockettransform_location 0.25) (+ (Transformation|GetSocketLocation (Variables|/Script/Engine.Character|GetMesh _asfort_pawn) "FX_Chest") (* _returnvalue_4 (Math|Vector|VectorLength (- _location (Transformation|GetSocketLocation (Variables|/Script/Engine.Character|GetMesh _asfort_pawn) "FX_Chest")))))) (Math|Rotator|MakeRotfromXY (Math|Vector|CrossProduct (Math|Vector|CrossProduct (Math|Vector|Normalize (- _outsockettransform_location (Transformation|GetSocketLocation (Variables|/Script/Engine.Character|GetMesh _asfort_pawn) "FX_Chest"))) (Math|Rotator|GetRotationXVector _outsockettransform_rotation)) (Math|Vector|Normalize (Math|Rotator|GetRotationXVector (Transformation|GetSocketRotation (Variables|/Script/Engine.Character|GetMesh _asfort_pawn) "FX_Combat")))) (Math|Rotator|GetRotationXVector (Transformation|GetSocketRotation (Variables|/Script/Engine.Character|GetMesh _asfort_pawn) "FX_Combat"))) false false)
            (Effects|Components|ParticleSystem|SetVectorParameter SpawnedPSC "PoofLocation" (- _location (+ (Transformation|GetSocketLocation (Variables|/Script/Engine.Character|GetMesh _asfort_pawn) "FX_Chest") (* _returnvalue_4 (Math|Vector|VectorLength (- _location (Transformation|GetSocketLocation (Variables|/Script/Engine.Character|GetMesh _asfort_pawn) "FX_Chest"))))))))
          (:Is%20Not%20Valid
            _)))
      (:Is%20Not%20Valid
        _))))

(event EventOnEquippedWeaponDestory
  ;[0.725 0.6854 0.4277] Show fx for weapon destruction on depleted durability
  (if (Variables|Effects|WeaponDurability|GetUseDestroyEffect)
    (Variables|Effects|WeaponDurability|SetDestroyEffectIsNiagara (Utilities|IsValid (Variables|Effects|WeaponDurability|GetWeaponDurabilityDestroyNS) :!class /Script/Engine.KismetSystemLibrary))
    (bind _outsockettransform (CallFunction|GetWeaponMeshSocketTransform (Variables|Effects|WeaponDurability|GetWeaponDurabilityDestroySocket)))
    (bind (_outsockettransform_location _outsockettransform_rotation) (Math|Transform|BreakTransform _outsockettransform))
    (Effects|Components|ParticleSystem|SpawnFXAtLocation (select (Variables|Effects|WeaponDurability|GetDestroyEffectIsNiagara) (Variables|Effects|WeaponDurability|GetWeaponDurabilityDestroyNS) (Variables|Effects|WeaponDurability|GetWeaponDurabilityDestroyEffect)) _outsockettransform_location (select (Variables|Effects|WeaponDurability|GetWeaponDurabilityDestroyUseSocketRotation) _outsockettransform_rotation "0, 0, 0"))
    (Variables|Effects|WeaponDurability|SetDestroyEffectIconIsNiagara (Utilities|IsValid (Variables|Effects|WeaponDurability|GetWeaponDurabilityDestroyIconNS) :!class /Script/Engine.KismetSystemLibrary))
    (bind _asfort_pawn (Utilities|Casting|CastToFortPawn (Game|GetInstigator)))
    (bind _returnvalue (Effects|SpawnFXSystemAttached :FXSystemAsset (select (Variables|Effects|WeaponDurability|GetDestroyEffectIconIsNiagara) (Variables|Effects|WeaponDurability|GetWeaponDurabilityDestroyIconNS) (Variables|Effects|WeaponDurability|GetWeaponDurabilityDestroyEffectIcon)) :AttachToComponent (Variables|/Script/Engine.Character|GetMesh _asfort_pawn) :AttachPointName (Variables|Effects|WeaponDurability|GetWeaponDurabilityDestroyIconAttachSocket) :Location (Variables|Effects|WeaponDurability|GetWeaponDurabilityDestroyIconLocationOffset) :Scale "1.000000,1.000000,1.000000" :LocationType "SnapToTarget" :bAutoDestroy true :PoolingMethod "None" :bAutoActivate true))
    (bind _returnvalue_2 (Rendering|Material|CreateDynamicMaterialInstance _returnvalue 0 "/Game/Effects/WeaponDestroy/M_WeaponDestroyExclamation.M_WeaponDestroyExclamation" "None" :!class /Script/Engine.PrimitiveComponent))
    (Rendering|Material|SetMaterial _returnvalue 0 _returnvalue_2)
    (Rendering|Material|SetTextureParameterValue _returnvalue_2 "IconTex" (ItemPreviewImage|GetorLoadSynchronously (Item|GetSmallPreviewImage (Data|GetWeaponData))))))

(event EventOnWeaponVisibilityChanged
  (params (bVisible :type Boolean) (bSetForLocalControllerOnly :type Boolean))
  (bind _alteration_ambient_ps (Variables|Default|GetAlteration%20Ambient%20PS))
  ;[1 1 1] EVENT -> OnWeaponVisibilityChanged
  (Utilities|IsValid (Variables|Default|GetAlteration%20Ambient%20PS)
    (:Is%20Valid
      (if bVisible
        (Components|Activation|Activate _alteration_ambient_ps true)
        (else
          (Components|Activation|Deactivate _alteration_ambient_ps))))))

(fn ActivateSwingFX
  (Networking|CanExecuteCosmeticEvents
    (:True
      (Utilities|IsValid (Variables|Default|GetSwingFX%20Niagara)
        (:Is%20Valid
          (Components|Activation|Activate (Variables|Default|GetSwingFX%20Niagara) true))))))

(fn BindSwingandAnimTrailEvents
  (Networking|CanExecuteCosmeticEvents
    (:True
      (bind _asmelee_swing_and_anim_trail_events_player_interface (Utilities|Casting|CastToMeleeSwingAndAnimTrailEventsPlayerInterface (Game|GetInstigator)))
      (Utilities|FlowControl|Sequence
        (:then_0
          ;[1 1 1] Bind Swing Trail Events for PlayerPawn _Athena
          (Utilities|IsValid (Niagara|GetNiagaraSystemAsset (Variables|Default|GetAnimTrail%20Niagara))
            (:Is%20Valid
              (MeleeSwingAndAnimTrailEventsPlayerInterface|BindtoAnimTrailEvents _asmelee_swing_and_anim_trail_events_player_interface (EventDispatchers|CreateEvent :!event On%20Anim%20Trails%20Notify) (EventDispatchers|CreateEvent :!event On%20Anim%20Trails%20Notify%20%28Deactivate%29) :!class /Script/FortniteGame.MeleeSwingAndAnimTrailEventsPlayerInterface))) ; If we actually have a Niagara system to play...
        )
        (:then_1
          ;[1 1 1] Bind Swing Events for PlayerPawn_Athena (Left, Right, Left End, and Right End)
          (Utilities|IsValid (Niagara|GetNiagaraSystemAsset (Variables|Default|GetSwingFX%20Niagara))
            (:Is%20Valid
              (MeleeSwingAndAnimTrailEventsPlayerInterface|BindtoSwingFX _asmelee_swing_and_anim_trail_events_player_interface (EventDispatchers|CreateEvent :!event OnSwingLeft) (EventDispatchers|CreateEvent :!event OnSwingLeftEnded) (EventDispatchers|CreateEvent :!event OnSwingRight) (EventDispatchers|CreateEvent :!event OnSwingRightEnded) :!class /Script/FortniteGame.MeleeSwingAndAnimTrailEventsPlayerInterface))))))) ; No need to register on the server, as we\nonly need these events for cosmetic purposes.
)

(fn DeactivateSwingFX
  (Components|Activation|Deactivate (Variables|Default|GetSwingFX%20Niagara)))

(fn GetInWater
  (return false))

(fn GetWeaponMeshSocketTransform
  (params (InSocketName :type Name))
  (bind _weaponmesh (Variables|Visual|GetWeaponMesh)
    (:then
      (Utilities|IsValid (Components|SkeletalMesh|GetSkeletalMeshAsset _weaponmesh)
        (:Is%20Valid
          (return (Transformation|GetSocketTransform _weaponmesh InSocketName "RTS_World")))
        (:Is%20Not%20Valid
          (return (Transformation|GetWorldTransform _weaponmesh)))))
    (:else
      (return (Math|Transform|MakeTransform "0, 0, 0")))))

(fn Melee_Effect_Color
  (return (Variables|Effects|GetEffects_Color_Level)))

(fn PlayCQCPickaxeEnemyAudio
  (params (Hit%20Result :type Hit%20Result%20Structure))
  (bind (_bblockinghit _binitialoverlap _time _distance _location _impactpoint _normal _impactnormal _physmat _hitactor) (Collision|BreakHitResult Hit%20Result))
  (Utilities|Casting|CastToBuildingActor _hitactor)
  (bind _asfort_pawn (Utilities|Casting|CastToFortPawn (Game|GetInstigator)))
  (switch Utilities|FlowControl|Switch|SwitchonEFortTeamAffiliation (Team|GetTeamAffiliationtoLocalPlayer (Team|GetActorTeam _asfort_pawn))
    (:Hostile
      (Utilities|IsValid (Pawn|GetLocalViewingPlayerController _asfort_pawn)
        (:Is%20Not%20Valid
          (Audio|PlaySoundatLocation (Variables|Sounds|GetCQC_EnemyPickaxe_Sound) _impactpoint "0, 0, 0" 1.0 1.0 0.0 0 0 self))))))

(fn SetActiveAlterationIdleParticles
  (params (Active :type Boolean) (Reset :type Boolean))
  (Networking|CanExecuteCosmeticEvents
    (:True
      (bind _alteration_ambient_ps (Variables|Default|GetAlteration%20Ambient%20PS)
        (:then
          (Components|Activation|SetActive _alteration_ambient_ps Active Reset)
          (return))
        (:else)))))

(fn SetWpnRarity
  ;[1 1 1] Wpn Rarity
  (Development|PrintString (Utilities|String|ToString(Float) (Math|Conversions|ToFloat(Byte) (Utilities|Select :Index (Item|GetRarity (Data|GetWeaponData)) :Common 0 :Uncommon 1 :Rare 2 :Epic 3 :Legendary 4 :Mythic 5 :Transcendent 6 :Unattainable 7)))))

(fn UnbindSwingAndAnimTrailEvents
  (bind _asmelee_swing_and_anim_trail_events_player_interface (Utilities|Casting|CastToMeleeSwingAndAnimTrailEventsPlayerInterface (Game|GetInstigator)))
  (MeleeSwingAndAnimTrailEventsPlayerInterface|UnbindfromAnimTrailEvents _asmelee_swing_and_anim_trail_events_player_interface (EventDispatchers|CreateEvent :!event On%20Anim%20Trails%20Notify) (EventDispatchers|CreateEvent :!event On%20Anim%20Trails%20Notify%20%28Deactivate%29) :!class /Script/FortniteGame.MeleeSwingAndAnimTrailEventsPlayerInterface)
  (MeleeSwingAndAnimTrailEventsPlayerInterface|UnbindfromSwingFX _asmelee_swing_and_anim_trail_events_player_interface (EventDispatchers|CreateEvent :!event OnSwingLeft) (EventDispatchers|CreateEvent :!event OnSwingLeftEnded) (EventDispatchers|CreateEvent :!event OnSwingRight) (EventDispatchers|CreateEvent :!event OnSwingRightEnded) :!class /Script/FortniteGame.MeleeSwingAndAnimTrailEventsPlayerInterface))

(fn ConstructionScript
  ;[1 1 1] SetupMeleeFXbyTag AnimNotify_PlayFX_WeapType uses this info
  (Effects|Components|ParticleSystem|SetTemplate (Variables|Default|GetMeleeHeavy_PSC) (Variables|Effects|GetMeleeHeavy_ParticleSystem))
  (Niagara|SetNiagaraSystemAsset (Variables|Default|GetIdleFX%20Niagara) (Variables|Effects|GetIdleFX%20Niagara%20Template) true)
  (Niagara|SetNiagaraSystemAsset (Variables|Default|GetSwingFX%20Niagara) (Variables|Effects|GetSwingFX%20Niagara%20Template) true)
  (Niagara|SetNiagaraSystemAsset (Variables|Default|GetAnimTrail%20Niagara) (Variables|Effects|GetAnimTrail%20Niagara%20Template) true)
  (|CallParentFunction))
