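; HeatSnake grapple weapon Blueprint, built on ChainWhip assets, handling equip effects, targeting sounds and HUD actions.
; It replaces the inherited swinging component with a grapple targeting component, which accepts targets beyond a minimum range.
; Melee and grapple targeting events update reticle and anim Blueprint, toggling a valid-target gameplay effect that gates secondary movement alongside cooldown.
; The server stores validated grapple hits, while an async cosmetic trace scales chain attack length.
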(event EventBeginPlay
  (|CallParentFunction)
  (Components|Activation|Deactivate (Variables|Default|GetGrappleSwingingComponent)) ; We want to override the GrappleSwingingComponent, \nbut we inherit it so for now just deactivate the inherited version.
  (Variables|Default|SetGrappleSwingingComp (Variables|Default|GetB_HeatSnakeGrappleTargetingComponent)))

(event FortWeapon|EventOnEquippedAndAttached
  ;[0.02491 0.1042 0.02354] On Equip: Set up HUD key actions (shown on the left hand side of the screen) and attach the weapon mesh to the desired socket on the hand of the pawn.
  (|CallParentFunction)
  (bind _ns_chainwhip_melee_trail (Variables|Default|GetNS_ChainWhip_Melee_Trail)
    (:then
      (Components|Activation|Activate _ns_chainwhip_melee_trail false))
    (:else
      _))
  (CallFunction|UpdateHudActions false)
  (bind _asfort_pawn (Utilities|Casting|CastToFortPawn (Game|GetInstigator)))
  (bind _output_get (Variables|Default|SetAs%20Fort%20Pawn _asfort_pawn))
  ;;[0.02934 0.1667 0.16] Attach the weapon mesh to the appropriate player mesh socket.
  (bind _mesh (Variables|/Script/Engine.Character|GetMesh _output_get)
    (:then
      (Transformation|AttachComponentToComponent (Variables|Visual|GetWeaponMesh) (Attachment|GetSceneComponenttoAttachWeaponTo _asfort_pawn) (Variables|WeaponPositionInfo|GetMainhandSocketName) "SnapToTarget" "SnapToTarget" "KeepWorld" true)
      ;;[0.15 0.01265 0 0.5] Execute equip screen FX
      (GameplayCue|ExecuteGameplayCueOnActor(Burst) (Variables|Default|GetAs%20Fort%20Pawn) (Variables|Default|GetEquipGameplayCueTag) (Ability|GameplayCue|MakeGameplayCueParameters 0.0)))
    (:else)))

(event FortWeapon|EventOnUnEquippedAndDetached
  ;[0.0243 0.165 0.1667] OnUnequippedAndDetached: Fade out the looping idle sound and disable/remove the camera modifier.
  (|CallParentFunction)
  (Utilities|FlowControl|Sequence
    (:then_0
      (Audio|SpawnSoundAttached "/SunRoseChainWhipGameplay/Items/ChainWhip/Sound/ChainWhip_UnEquip_Cue.ChainWhip_UnEquip_Cue" (Variables|/Script/Engine.Actor|GetRootComponent (Game|GetInstigator)) "hand_r"))
    (:then_1
      (Networking|CanExecuteCosmeticEvents))
    (:then_2
      (bind _ns_chainwhip_melee_trail (Variables|Default|GetNS_ChainWhip_Melee_Trail)
        (:then
          (Components|Activation|Deactivate _ns_chainwhip_melee_trail))
        (:else)))
    (:then_3
      (TouchControls|RestoreTouchScreenRegions))))

(event Custom|UpdateAttackRangeAnimBP
  (bind _returnvalue (GameInstanceSubsystems|GetTargetingSubsystem))
  ;[0.09147 0.03613 0.1094] UpdateAttackRangeAnimBP: Perform a targeting trace to figure out the distance of the trace in the aiming direction.
  ; This value is used in the animation blueprint to scale the length of the chain on the attack.
  ; Note: This logic needs to run on simulated and autonomous proxies.
  (Networking|CanExecuteCosmeticEvents
    (:True
      (bind _b_heatsnaketargetingcomponent (Variables|Default|GetB_HeatSnakeTargetingComponent)
        (:then
          (Utilities|IsValid (General|GetPlayerPawn _b_heatsnaketargetingcomponent :!class /HeatSnake/Gameplay/Targeting/B_HeatSnakeTargetingComponent.B_HeatSnakeTargetingComponent_C)
            (:Is%20Valid
              (bind (_asynctaskproxy _targetinghandle) (Class|AsyncActionPerformTargeting|PerformTargetingAsyncAction (General|GetPlayerPawn _b_heatsnaketargetingcomponent :!class /HeatSnake/Gameplay/Targeting/B_HeatSnakeTargetingComponent.B_HeatSnakeTargetingComponent_C) "/HeatSnake/Gameplay/Targeting/TP_HeatSnake_HookshotTargeting.TP_HeatSnake_HookshotTargeting")
                (:then
                  (Variables|Transient|SetAsyncTask_AttackRangeTrace _asynctaskproxy))
                (:Targeted
                  (Utilities|IsValid _returnvalue
                    (:Is%20Valid
                      (if (Utilities|Array|IsValidIndex (TargetingSystem|TargetingResults|GetTargetingResults _returnvalue _targetinghandle) (Utilities|Array|LastIndex (TargetingSystem|TargetingResults|GetTargetingResults _returnvalue _targetinghandle)))
                        (Variables|Default|SetLastTargetData (Utilities|Array|Get(aref) (TargetingSystem|TargetingResults|GetTargetingResults _returnvalue _targetinghandle) (Utilities|Array|LastIndex (TargetingSystem|TargetingResults|GetTargetingResults _returnvalue _targetinghandle))))
                        (if (Collision|BreakHitResult (Utilities|Array|Get(aref) (TargetingSystem|TargetingResults|GetTargetingResults _returnvalue _targetinghandle) (Utilities|Array|LastIndex (TargetingSystem|TargetingResults|GetTargetingResults _returnvalue _targetinghandle))))
                          ;;[0.03005 0.151 0.1253] Update the attack range based on the distance result of the trace.
                          (bind _animbp (Utilities|GetWeaponAnimBP)
                            (:Valid
                              (Class|ChainsSunRoseSharpWepAnimBP|UpdateAttackRange _animbp (.Distance (Collision|BreakHitResult (Utilities|Array|Get(aref) (TargetingSystem|TargetingResults|GetTargetingResults _returnvalue _targetinghandle) (Utilities|Array|LastIndex (TargetingSystem|TargetingResults|GetTargetingResults _returnvalue _targetinghandle))))))))
                          (else
                            _))
                        (else
                          _))
                      ;;[0.06385 0.194 0.1979] Any failure to perform the trace should update the attack range to max distance.
                      (bind _animbp_2 (Utilities|GetWeaponAnimBP)
                        (:Valid
                          (Class|ChainsSunRoseSharpWepAnimBP|UpdateAttackRange _animbp_2 (ScalableFloat|GetValueatLevel (Variables|Hotfix|GetHotfix_MeleeAttackRange))))))
                    (:Is%20Not%20Valid
                      _)))))
            (:Is%20Not%20Valid
              _)))
        (:else
          _)))))

(event OnTargetingHookshot(B_HeatSnakeTargetingComponent)
  (params (bTargetingHookshot :type Boolean))
  (bind _returnvalue (Game|GetInstigator))
  (bind _soundlibrarycomponent (Variables|SoundLibrary|GetSoundLibraryComponent))
  (bind _weaponequipsoundcomponent (Variables|Default|GetWeaponEquipSoundComponent))
  ;[0.2917 0.161 0.111] OnTargetingHookshot: Play the appropriate sound when the weapon is entering and exiting the hookshot targeting state.
  (Networking|CanExecuteCosmeticEvents
    (:True
      (if bTargetingHookshot
        ;;[0.1976 1 0.9801] This event is expensive (~300 microseconds) and will occur during frames where the chainwhip is already expensive, so we delay 1 frame to avoid a spike in cost.
        (Utilities|FlowControl|DelayUntilNextTick)
        (SoundLibrary|PlayEvent _soundlibrarycomponent (Variables|Sounds|GetTargetingSoundLibraryTag))
        (Class|WeaponEquipSoundComponent|StopAudio _weaponequipsoundcomponent)
        (CallFunction|UpdateHudActions true)
        (else
          (Utilities|FlowControl|DelayUntilNextTick)
          (SoundLibrary|PlayEvent _soundlibrarycomponent (Variables|Sounds|GetTargetingExitSoundLibraryTag))
          (Class|WeaponEquipSoundComponent|PlayAudio _weaponequipsoundcomponent)
          (CallFunction|UpdateHudActions false)
          ;;[0.1531 0.1804 1] If the player is holding down the primary input when targeting ends (and NOT in a vehicle), perform a melee attack.
          (bind _asfort_player_pawn (Utilities|Casting|CastToFortPlayerPawn _returnvalue))
          (if (Math|Boolean|ANDBoolean (Variables|/Script/FortniteGame.FortPawn|GetbPrimaryInputHeld _asfort_player_pawn) (Pawn|IsLocallyControlled _asfort_player_pawn) (not (Vehicle|IsinVehicle _asfort_player_pawn)))
            (Mantis|SimulateInputEvent (Actor|GetComponentbyClass _returnvalue "/Script/MantisRuntime.FortMantisPawnComponent") "Primary" "Pressed" true false)))))))

(event EventEndPlay
  (params (EndPlayReason :type EEndPlayReason%20Enum :default "Destroyed"))
  ;[0.1198 0.06687 0.02099] EndPlay: Clean up the added camera modifier.
  (Utilities|FlowControl|Sequence
    (:then_0
      (Networking|CanExecuteCosmeticEvents))
    (:then_1
      (|CallParentFunction EndPlayReason))))

(event Custom|GrantTargetingHandler
  (params (Apply :type Boolean))
  (bind _as_fort_pawn (Variables|Default|GetAs%20Fort%20Pawn))
  (bind _returnvalue (Ability|GetFortAbilitySystemComponent _as_fort_pawn))
  ;[0.15 0.02666 0.1333 0.5] SERVER: Grant Player Valid Target Effect Tag
  (if Apply
    (bind _returnvalue_2 (GameplayEffects|ApplyGameplayEffectToSelf :self _returnvalue :GameplayEffectClass (Variables|Default|GetHasValidTarget) :Level 0.0))
    (Variables|Default|SetHasValidTargetEffectHandle _returnvalue_2)
    (else
      (GameplayEffects|RemoveActiveEffectswithGrantedTags _returnvalue "(GameplayTags=((TagName=\"Granted.HeatSnake.Targeting.HaveValidTarget\")))"))))

(event OnTargetInfoUpdated(B_HeatSnakeTargetingComponent)
  (params (HitResult :type Hit%20Result%20Structure) (TargetType :type ChainWhipTargetActorTypeEnum%20Enum :default "NewEnumerator0"))
  (bind (_bblockinghit _binitialoverlap _time _distance _location _impactpoint) (Collision|BreakHitResult HitResult))
  (bind _returnvalue (Utilities|Enum|Equal(Enum) TargetType "NewEnumerator2"))
  (bind _returnvalue_2 (Utilities|Enum|Equal(Enum) TargetType "NewEnumerator1"))
  (bind _returnvalue_3 (or _returnvalue_2 _returnvalue))
  ; Set the location of the Reticle on a MeleeTarget
  (bind _output_get (Variables|Default|SetbMeleeTargetValid _returnvalue_3))
  (if _output_get
    (Utilities|FlowControl|Sequence
      (:then_0
        (Class|FortWeaponComponentTargetState|UpdateTargetInfo (Variables|Default|GetFortWeaponComponent_TargetState) _returnvalue_3 _impactpoint)
        ;;[0.15 0.1172 0 0.5] Hack. This could potentially return an invalid object!
        (bind _heatsnakeanimbp (CallFunction|GetHeatSnakeAnimBP))
        (Class|FridayPloyHeatSnakeSharpSkeletonAnimBlueprint|UpdateTargetingInfo _heatsnakeanimbp _returnvalue_3 HitResult))
      (:then_1
        _)) ; Movement Evaluation Should still occur
    (else
      _))
  (if (Variables|Default|GetbGrappleTargetValid)
    (if (GameplayTags|HasMatchingGameplayTag (Variables|Default|GetAs%20Fort%20Pawn) "(TagName=\"Granted.Cooldown.HeatSnake.Movement\")")
      (CallFunction|GrantTargetingHandler false) ; Remove Gameplay Effect from Player when on Cooldown\nWithout the Tag granted, the secondary movement WILL NOT ACTIVATE
      (elif (not (CallFunction|HasValidTargetTag))
        (CallFunction|GrantTargetingHandler true) ; Apply Gameplay Effect to Player when there is an active Target\nWithout the Tag granted, the secondary movement WILL NOT ACTIVATE
      ))
    (else
      (Class|FortWeaponComponentTargetState|ClearTrackedTarget (Variables|Default|GetFortWeaponComponent_TargetState))
      ;;[0.15 0.1421 0 0.5] Hack. This could potentially return an invalid object!
      (bind _heatsnakeanimbp_2 (CallFunction|GetHeatSnakeAnimBP))
      (Class|FridayPloyHeatSnakeSharpSkeletonAnimBlueprint|UpdateTargetingInfo :self _heatsnakeanimbp_2 :HasValidTarget false)
      (if (CallFunction|HasValidTargetTag)
        (CallFunction|GrantTargetingHandler false) ; Remove Gameplay Effect from Player when there is NO active Target\nWithout the Tag granted, the secondary movement WILL NOT ACTIVATE
      ))))

(event OnTargetInfoUpdated(B_HeatSnakeGrappleTargetingComponent)
  (params (HitResult :type Hit%20Result%20Structure) (TargetType :type HeatSnakeGrappleTargetTypeEnum%20Enum :default "NewEnumerator0"))
  (bind (_bblockinghit _binitialoverlap _time _distance _location _impactpoint) (Collision|BreakHitResult HitResult))
  (bind _returnvalue (CallFunction|HasValidMoveDistance _impactpoint))
  (bind _returnvalue_2 (and (Utilities|Enum|Equal(Enum) TargetType "NewEnumerator1") _returnvalue))
  (bind _returnvalue_3 (Utilities|Enum|Equal(Enum) TargetType "NewEnumerator1"))
  ; Set the location of the Reticle on the grapple target
  (if _returnvalue_2
    (Variables|Default|SetbGrappleTargetValid _returnvalue_2)
    (Class|FortWeaponComponentTargetState|UpdateTargetInfo (Variables|Default|GetFortWeaponComponent_TargetState) (Variables|Default|GetbGrappleTargetValid) _impactpoint)))

(event HeatSnakeWeapon|EventSetTetherPointData
  (params (NewTetherPoint :type Vector :default "0, 0, 0"))
  (Variables|Default|SetTetherPoint NewTetherPoint))

(event Custom|UpdateGrappleReticleFromNative
  (params (HitSomething :type Boolean) (BlockedUse :type Boolean) (IsTargetatAimLocation :type Boolean) (TargetPosition :type Vector))
  (bind _b_heatsnakegrappletargetingcomponent (Variables|Default|GetB_HeatSnakeGrappleTargetingComponent))
  (bind _targetinghitresult (Variables|/Script/FortniteGame.FortWeaponComponent_Swinging|GetTargetingHitResult _b_heatsnakegrappletargetingcomponent))
  (if (Math|Boolean|ANDBoolean HitSomething (not BlockedUse) (CallFunction|HasValidMoveDistance TargetPosition))
    (Reticle|Targeting|SetTargetType _b_heatsnakegrappletargetingcomponent "NewEnumerator1" :!class /HeatSnake/Gameplay/Targeting/B_HeatSnakeGrappleTargetingComponent.B_HeatSnakeGrappleTargetingComponent_C)
    (Reticle|Targeting|UpdateTargetInformation _b_heatsnakegrappletargetingcomponent _targetinghitresult "NewEnumerator1" :!class /HeatSnake/Gameplay/Targeting/B_HeatSnakeGrappleTargetingComponent.B_HeatSnakeGrappleTargetingComponent_C)
    (else
      (Reticle|Targeting|SetTargetType _b_heatsnakegrappletargetingcomponent "NewEnumerator0" :!class /HeatSnake/Gameplay/Targeting/B_HeatSnakeGrappleTargetingComponent.B_HeatSnakeGrappleTargetingComponent_C)
      (Reticle|Targeting|UpdateTargetInformation _b_heatsnakegrappletargetingcomponent _targetinghitresult "NewEnumerator0" :!class /HeatSnake/Gameplay/Targeting/B_HeatSnakeGrappleTargetingComponent.B_HeatSnakeGrappleTargetingComponent_C))))

(event OnServerTargetValidated(B_HeatSnakeGrappleTargetingComponent)
  (if (and (not (GameplayTags|HasMatchingGameplayTag (Variables|Default|GetAs%20Fort%20Pawn) "(TagName=\"Granted.Cooldown.HeatSnake.Movement\")")) (GameplayTags|HasMatchingGameplayTag (Variables|Default|GetAs%20Fort%20Pawn) "(TagName=\"Granted.HeatSnake.Targeting.HaveValidTarget\")"))
    (Networking|SwitchHasAuthority
      (:Authority
        (Variables|Default|SetValidatedServerGrappleHitInfo (Variables|/Script/FortniteGame.FortWeaponComponent_Swinging|GetTargetingHitResult (Variables|Default|GetB_HeatSnakeGrappleTargetingComponent)))))))

(fn GetCameraModifier
  (return (Variables|Camera|GetHeatSnake_CameraModifier)))

(fn GetHeatSnakeAnimBP
  (bind _heatsnakeanimbp (Variables|Default|GetHeatSnakeAnimBP)
    (:then
      (return _heatsnakeanimbp))
    (:else
      (Utilities|FlowControl|DoOnce false
        (:Completed
          (bind _asfort_player_pawn (Utilities|Casting|CastToFortPlayerPawn (Game|GetInstigator)))
          (bind _array_element (Utilities|Array|ForEachLoopwithBreak (Components|GetChildrenComponents (Variables|/Script/Engine.Character|GetMesh _asfort_player_pawn) false))
            (:LoopBody
              (bind _asskeletal_mesh_component (Utilities|Casting|CastToSkeletalMeshComponent _array_element))
              (bind _asfriday_ploy_heat_snake_sharp_skeleton_anim_blueprint (Utilities|Casting|CastToFridayPloy_HeatSnake_Sharp_Skeleton_AnimBlueprint (Components|SkeletalMesh|GetAnimInstance _asskeletal_mesh_component)))
              (Variables|Default|SetHeatSnakeAnimBP _asfriday_ploy_heat_snake_sharp_skeleton_anim_blueprint))
            (:Completed
              _))
          (return (Variables|Default|GetHeatSnakeAnimBP)))))))

(fn GetHookshotProjectile
  ;[0.04513 0.2031 0.1829] Return the reference to the latest hookshot projectile. Can be null. Client always returns null.
  (return (Variables|Hookshot|GetHookshotProjectile)))

(fn GetTargetData
  (return (Variables|Default|GetLastTargetData)))

(fn GetTargetingResult
  (params (GettingFireTarget :type Boolean))
  (return (Variables|/Script/FortniteGame.FortWeaponComponent_Swinging|GetTargetingHitResult (Variables|Default|GetGrappleSwingingComponent))))

(fn GetTetherPointData
  (return (Variables|Default|GetTetherPoint)))

(fn HasValidMoveDistance
  (params (V1 :type Vector))
  (return (>= (Math|Vector|Distance(Vector) V1 (Transformation|GetActorLocation (Variables|Default|GetAs%20Fort%20Pawn))) (ScalableFloat|GetValueatLevel (Variables|Hotfix|GetHotfix_MovementMinRange)))))

(fn HasValidTargetTag
  (return (GameplayTags|HasMatchingGameplayTag (Ability|GetFortAbilitySystemComponent (Variables|Default|GetAs%20Fort%20Pawn)) "(TagName=\"Granted.HeatSnake.Targeting.HaveValidTarget\")")))

(fn RestoreTouchScreenRegions
  (bind _returnvalue (Game|GetInstigator))
  (bind _returnvalue_2 (Fortnite|GetFortPlayerControllerfromActor _returnvalue))
  (bind _returnvalue_3 (Development|GetConsoleVariableBoolValue "HeatSnakeTouchControlsImprovement"))
  (if _returnvalue_3
    (if (Input|IsUsingTouch _returnvalue_2)
      (TouchControls|SetTouchRegionBounds :self _returnvalue_2 :MovePlayerBounds (Math|Vector4|MakeVector4 0.0 0.5 0.0 1.0) :RotateCameraBounds (Math|Vector4|MakeVector4 0.5 1.0 0.0 1.0)))))

(fn SetHiddenMeleeTrail
  (params (NewHidden :type Boolean))
  ;[0.05441 0.126 0.1354] SetHiddenMeleeTrail: Hide/Unhide the melee trail component when desired.
  (bind _ns_chainwhip_melee_trail (Variables|Default|GetNS_ChainWhip_Melee_Trail)
    (:then
      (Development|SetHiddeninGame _ns_chainwhip_melee_trail NewHidden false))
    (:else)))

(fn SetHookshotProjectile
  (params (HookshotProjectile :type B%20Prj%20Chain%20Whip%20Hookshot%20V%202%20Object%20Reference))
  ;[0.03895 0.1819 0.1875] SetHookshotProjectile: Called to store the ref to the projectile so that it can be accessed from other locations.
  ; Called from GA_ChainWhip_Secondary_Hookshot.
  (Networking|SwitchHasAuthority
    (:Authority
      (Variables|Hookshot|SetHookshotProjectile HookshotProjectile))))

(fn UpdateHudActions
  (params (InTargetingMode :type Boolean))
  (WeaponHUDData|SetHudKeyActionsVisibilitybyID (Utilities|Array|MakeArray (Utilities|Struct|MakeWeaponHudKeyActionIdandVisibility "ChainWhip_Hookshot_Fire" InTargetingMode) (Utilities|Struct|MakeWeaponHudKeyActionIdandVisibility "ChainWhip_Attack" (not InTargetingMode)) (Utilities|Struct|MakeWeaponHudKeyActionIdandVisibility "ChainWhip_Hookshot_Target" (not InTargetingMode)))))

(fn ConstructionScript
  (|CallParentFunction))
